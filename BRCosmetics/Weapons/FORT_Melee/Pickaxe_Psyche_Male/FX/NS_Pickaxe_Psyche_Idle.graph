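# Niagara idle effect for the Psyche Male pickaxe, adding drifting goop globs and flowing ribbons to the axe mesh.
# GoopDrop spawns tumbling mesh globs on the axe's dyn_glob bones that drift upward, with spawning cut to zero while User.BladeSwing is true.
# Continuous and burst ribbon emitters use FN_BoneFlow to wrap grouped particles along the BoneFlow sockets.
# Pendulum-constrained Hero_Drips emitters and the original Ribbon_BoneFlow emitter are present but disabled.
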NiagaraSystem /BRCosmetics/Weapons/FORT_Melee/Pickaxe_Psyche_Male/FX/NS_Pickaxe_Psyche_Idle
dependencies {
 usedDataInterfaces [&NiagaraDataInterfaceCurve, &NiagaraDataInterfaceVectorCurve]
 usedDynamicInputs []
 usedModules [&SystemState]
}
emitters {
 GoopDrop {
  emitterUpdateScript [
   &EmitterState(Inactive Response=::NewEnumerator1, Life Cycle Mode=::NewEnumerator1, Loop Duration=2, UseLoopDelay=false)
   &SpawnRate(
    SpawnRate=&Lerp_Float(
     A=&RandomRangeFloat(Minimum=2, Maximum=10, Evaluation Type=::NewEnumerator1)
     B=0
     Alpha=&MakeCustomFloatFromBool(Bool=$User.BladeSwing, True Float=1, False Float=0)
    )
    Use Spawn Probability=false
   )
  ]
  particleSpawnScript [
   &InitializeParticle(
    Lifetime Max=0.8
    Lifetime Min=0.3
    Lifetime Mode=::NewEnumerator1
    Mesh Scale Max=<1.2 1.2 1.2>
    Mesh Scale Min=<0.5 0.5 0.5>
    Mesh Scale Mode=::NewEnumerator2
    UsePositionOffset=false
    Write Material Random=true
   )
   &Color(Color=&MakeLinearColorFromVectorAndFloat("Vector (RGB)"=<1 1 1>, "Float (Alpha)"=&ClampFloat(Float=1))) {enabled=false}
   &SkeletalMeshLocation(
    Bone Sampling Mode=::NewEnumerator0
    MaskSpawnGroup=false
    OverrideRandomSeed=false
    Skeletal Mesh=DataInterface() {propertyValues {
      DefaultMesh=null
      ExcludeBoneName=root
      OptionalComponentClass=null
      PreviewMesh=/BRCosmetics/Weapons/FORT_Melee/Pickaxe_Psyche_Male/Meshes/Psyche_Male_Axe
      SkinningMode=SkinOnTheFly
      SoftSourceActor=null
      SourceMode=Default
      UvSetIndex=0
      WholeMeshLOD=-1
      bExcludeBone=false
      bReadDeformedGeometry=true
      bRequireCurrentFrameData=true
      ComponentTags []
      FilteredBones [dyn_globA_2,dyn_globB_2,dyn_globC_2,dyn_globD_2,dyn_globE_2,dyn_globF_2,dyn_globG_2,dyn_globH_2,dyn_globI_2,dyn_globJ_2,dyn_globK_2,dyn_globL_2,dyn_globM_2]
      FilteredSockets []
      MeshUserParameter {parameter {name=None, typeDefHandle {registeredTypeIndex=1}, varData []}}
      SamplingRegions []
     }}
   ) {module=SkeletalMeshLocation001}
   &AddVelocity(Coordinate Space=::World, Velocity=<0 0 10>) {module=AddVelocity001}
   &InheritVelocity(Inherited Velocity Amount Scale=&VectorFromFloat(Value=0.5), Speed Limit=true, UseSpeedThreshold=false)
  ]
  particleUpdateScript [
   &ParticleState()
   &ScaleMeshSize(Scale Factor=&VectorFromFloat(Value=&FloatFromCurve(
      FloatCurve=DataInterface() {propertyValues {Curve {keys [
          {arriveTangent=9.1187, interpMode=RCIM_Cubic, time=0, value=0}
          {arriveTangent=0.3759, interpMode=RCIM_Cubic, leaveTangent=0.3759, time=0.25, value=1.1246}
          <at 0.7835 0.2945 RCIM_Cubic>
          <at 1 0 RCIM_Cubic>
         ]}}}
      Scale Curve=0.6
     ))) {module=ScaleMeshSize001}
   &ScaleColor(
    Scale Alpha=&FloatFromCurve(FloatCurve=DataInterface() {propertyValues {Curve {keys [<at 0.5 1.0113 RCIM_Cubic>,<at 1.0011 0.002 RCIM_Cubic>]}}})
    ScaleA=true
    ScaleRGB=false
   ) {enabled=false}
   &DynamicMaterialParameters(
    Index 0 Param 1=4
    Index 0 Param 2=5
    Index 0 Param 3=2
    Module.Param0WriteEnabled=true
    Module.Param1WriteEnabled=true
    Module.Param2WriteEnabled=true
    Module.Param3WriteEnabled=false
   ) {enabled=false}
   &DynamicMaterialParameters(
    Index 0 Param 1=&RandomRangeFloat(Minimum=0.5, Maximum=0.8)
    Index 0 Param 2=&RandomRangeFloat(Minimum=4, Maximum=5)
    Index 0 Param 3=&FloatFromCurve(FloatCurve=DataInterface() {propertyValues {Curve {keys [
         {arriveTangent=2.0663, interpMode=RCIM_Cubic, leaveTangent=2.0663, tangentMode=RCTM_User, time=0, value=0.2}
         {interpMode=RCIM_Cubic, tangentMode=RCTM_User, time=0.6, value=1}
        ]}}})
    Index 0 Param 4=&RandomRangeFloat(Minimum=50, Maximum=55, Evaluation Type=::NewEnumerator1)
    Module.Param0WriteEnabled=true
    Module.Param1WriteEnabled=true
    Module.Param2WriteEnabled=true
    Module.Param3WriteEnabled=false
   ) {module=DynamicMaterialParameters002}
   &AccelerationForce(Acceleration=&Multiply_VectorByFloat(
     Vector=&RandomRangeVector(Minimum=<-3 -3 5>, Maximum=<3 3 10>, Evaluation Type=::NewEnumerator1)
     Float=1
    ))
   &Drag(Drag=1.5, Use Linear Drag=true, Use Rotational Drag=false) {enabled=false}
   &InheritSourceMovement(
    Applied Position Scale=&VectorFromFloat(Value=&FloatFromCurve(FloatCurve=DataInterface() {propertyValues {Curve {keys [<at 0.1 1>,<at 0.15 0>]}}}))
    ApplyForce=false
    ApplyPosition=true
    ApplyPositionFalloff=false
   )
   &SolveForcesAndVelocity(Clamp Velocity=false, Limit Acceleration=false)
   &MeshRotationForce(
    Coordinate Space:ENiagaraMeshTransforms=::NewEnumerator3
    Rotation=&RandomRangeVector(Minimum=<-4 -4 -4>, Maximum=<4 4 4>)
   )
   &SolveRotationalForcesAndVelocity()
  ]
  rendererClasses [&NiagaraMeshRendererProperties]
  renderers [{rendererClass=&NiagaraMeshRendererProperties, rendererIndex=0}]
 }
 Hero_Drips {
  bEnabled=false
  emitterUpdateScript [
   &EmitterState()
   &SpawnBurst_Instantaneous(Spawn Count=3, Use Spawn Probability=true) {enabled=false}
   &SpawnRate(Spawn Probability=0.8, SpawnRate=1.5, Use Spawn Probability=true)
  ]
  particleSpawnScript [
   &FN_SetSpawnGroup()
   /SetVariables(
    Particles.PendulumDrag=1
    Particles.PendulumLengthScale:NiagaraFloat=&MaskFloatBySpawnGroup(
     Masked Float=&RandomRangeFloat(Minimum=0.35, Maximum=0.5)
     Passthrough Float=&RandomRangeFloat(Minimum=0.375, Maximum=0.415)
    )
    Particles.PendulumMaskFalloff=1
    Particles.PendulumMeshWidth:NiagaraFloat=&RandomRangeFloat(Minimum=0.1, Maximum=0.125)
    Particles.PendulumRigidity=0.06
   ) {isSetParameters=true, module=SetVariables_#1}
   /Niagara/Modules/Spawn/Initialization/InitializeParticle(
    Lifetime=&UniformRangedFloat(Minimum=2.2, Maximum=2)
    Mesh Scale=&MakeVector(X=$Particles.PendulumMeshWidth, Y=$Particles.PendulumMeshWidth, Z=$Particles.PendulumLengthScale)
    Module.Write Color=true
    Module.Write Lifetime=true
    Module.Write Mass=true
    Module.Write Material Random=true
    Module.Write Position=false
    Module.Write Scale=true
    Module.Write SpriteRotation=false
    Module.Write SpriteSize=true
   )
   &CameraOffset(Camera Offset Amount=2)
  ]
  particleUpdateScript [
   &ParticleState()
   &SkeletalMeshLocation(
    Apply To Particle Position=false
    Bone / Socket Index=$Particles.SpawnGroup
    Inherit Velocity=false
    MaskSpawnGroup=false
    Mesh Sampling Type=::NewEnumerator5
    Skeletal Mesh=$User.SkeletalMesh
    Socket Sampling Mode=::NewEnumerator5
   )
   /SetVariables(Particles.MeshOrigin:Vector3f=&Add_Vector(
     A=$Output.SkeletalMeshLocation.SampledPosition
     B=&RotateVectorByQuaternion(
      Vector To Rotate=&RandomRangeVector(Minimum=<-0.5 -0.5 -0.5>, Maximum=<0.5 0.5 0.5>)
      Quaternion=$Output.SkeletalMeshLocation.SampledRotation
     )
    )) {isSetParameters=true, module=SetVariables_#2}
   &Drag(Drag=$Particles.PendulumDrag, Use Linear Drag=true, Use Rotational Drag=false)
   &GravityForce()
   &PendulumSetup(
    Module.Enable Spring Driver=true
    Pendulum Length:NiagaraFloat=&Multiply_Float(A=100, B=$Particles.PendulumLengthScale)
    Pendulum Pivot:Vector3f=$Particles.MeshOrigin
    Pendulum Rest Axis:Vector3f=&MaskVectorBySpawnGroup(Masked Vector=<0 0 1>, Passthrough Vector=<0 0 -1>)
    "Spring Driven Constraint (Tightness)"=10
   )
   /SetVariables(Particles.MeshFacing:Vector3f=&NormalizeVector(Vector To Normalize=&Subtract_Vector(A=$Particles.MeshOrigin, B=Linked() {linkedVariable {name=Particles.Position, type=NiagaraPosition}}))) {isSetParameters=true, module=SetVariables_#3}
   &self:Pendulum_Facing(
    Bone Quat:Quat4f=$Output.SkeletalMeshLocation.SampledRotation
    FromVector:Vector3f=&MaskVectorBySpawnGroup(Masked Vector=<0 0 -1>, Passthrough Vector=<0 0 1>)
    ToVector:Vector3f=$Particles.MeshFacing
    X Vector:Vector3f=&MaskVectorBySpawnGroup(Masked Vector=<-1 0 0>, Passthrough Vector=<-1 0 0>)
    Y Vector:Vector3f=<0 0 0>
   ) {module=Pendulum_Facing}
   &SolveForcesAndVelocity(Clamp Velocity=false, Limit Acceleration=false)
   &PendulumConstraint()
   &CalculateAccurateVelocity()
   &DynamicMaterialParameters(
    Index 1 Param 3=1.0081
    Index 1 Param 4=&FloatFromCurve(FloatCurve=DataInterface() {propertyValues {Curve {keys [
         {arriveTangent=3.6645, interpMode=RCIM_Cubic, leaveTangent=3.6645, tangentMode=RCTM_User, time=0.1, value=-0.1}
         {interpMode=RCIM_Cubic, leaveTangent=2.3573, time=0.9, value=2.125}
        ]}}})
    Index 2 Param 1=HlslExpression() {hlslExpression=Particles.MeshOrigin.x}
    Index 2 Param 2=HlslExpression() {hlslExpression=Particles.MeshOrigin.y}
    Index 2 Param 3=HlslExpression() {hlslExpression=Particles.MeshOrigin.z}
    Index 2 Param 4=$Particles.PendulumMaskFalloff
    Index 3 Param 1=&MakeFloatFromVector(Vector=Linked() {linkedVariable {name=Particles.Position, type=NiagaraPosition}}, Channel=::NewEnumerator0)
    Index 3 Param 2=&MakeFloatFromVector(Vector=Linked() {linkedVariable {name=Particles.Position, type=NiagaraPosition}}, Channel=::NewEnumerator1)
    Index 3 Param 3=&MakeFloatFromVector(Vector=Linked() {linkedVariable {name=Particles.Position, type=NiagaraPosition}}, Channel=::NewEnumerator2)
    Index 3 Param 4=$Particles.PendulumRigidity
    Module.Param0WriteEnabled1=true
    Module.Param0WriteEnabled2=true
    Module.Param0WriteEnabled3=true
    Module.Param1WriteEnabled1=true
    Module.Param1WriteEnabled2=true
    Module.Param1WriteEnabled3=true
    Module.Param2WriteEnabled1=true
    Module.Param2WriteEnabled2=true
    Module.Param2WriteEnabled3=true
    Module.Param3WriteEnabled1=true
    Module.Param3WriteEnabled2=true
    Module.Param3WriteEnabled3=true
    Write Parameter Index 0=false
    Write Parameter Index 1=true
    Write Parameter Index 2=true
    Write Parameter Index 3=true
   )
   &ScaleMeshSize(Scale Factor=&MakeVector(
     X=&FloatFromCurve(FloatCurve=DataInterface() {propertyValues {Curve {keys [
          <at 0 0 RCIM_Cubic>
          {arriveTangent=2.5, interpMode=RCIM_Cubic, leaveTangent=2.5, time=0.05, value=0.9}
          {arriveTangent=-0.7353, interpMode=RCIM_Cubic, leaveTangent=-0.7353, time=0.6, value=0.65}
          {arriveTangent=-4.2857, interpMode=RCIM_Cubic, leaveTangent=-4.2857, time=0.9, value=0.375}
          {interpMode=RCIM_Cubic, leaveTangent=-0.6579, time=0.95, value=0}
         ]}}})
     Y=&FloatFromCurve(FloatCurve=DataInterface() {propertyValues {Curve {keys [
          <at 0 0 RCIM_Cubic>
          {arriveTangent=2.5, interpMode=RCIM_Cubic, leaveTangent=2.5, time=0.05, value=0.9}
          {arriveTangent=-0.7353, interpMode=RCIM_Cubic, leaveTangent=-0.7353, time=0.6, value=0.65}
          {arriveTangent=-4.2857, interpMode=RCIM_Cubic, leaveTangent=-4.2857, time=0.9, value=0.375}
          {interpMode=RCIM_Cubic, leaveTangent=-0.6579, time=0.95, value=0}
         ]}}})
     Z=&FloatFromCurve(FloatCurve=DataInterface() {propertyValues {Curve {keys [
          {arriveTangent=1.8796, interpMode=RCIM_Cubic, time=0, value=0}
          {arriveTangent=-4.7929, interpMode=RCIM_Cubic, leaveTangent=-4.7929, tangentMode=RCTM_User, time=0.625, value=0.85}
          {arriveTangent=-1.7696, interpMode=RCIM_Cubic, leaveTangent=-1.7696, tangentMode=RCTM_User, time=0.95, value=0}
         ]}}})
    ))
   /SetVariables(
    Particles.Drop Mesh Size:Vector3f=&VectorFromCurve(
     VectorCurve=DataInterface() {propertyValues {
       XCurve {keys [
         <at 0 1>
         {arriveTangent=-18.2164, interpMode=RCIM_Cubic, leaveTangent=-18.2164, tangentMode=RCTM_User, time=0.85, value=1}
         {arriveTangent=-2.286, interpMode=RCIM_Cubic, leaveTangent=-2.286, tangentMode=RCTM_User, time=0.8575, value=0}
        ]}
       YCurve {keys [
         <at 0 1>
         {arriveTangent=-18.2164, interpMode=RCIM_Cubic, leaveTangent=-18.2164, tangentMode=RCTM_User, time=0.85, value=1}
         {arriveTangent=-2.286, interpMode=RCIM_Cubic, leaveTangent=-2.286, tangentMode=RCTM_User, time=0.8575, value=0}
        ]}
       ZCurve {keys [
         <at 0.1 0.675>
         <at 0.3 0.8>
         <at 0.6 0.4>
         <at 0.8 0.8>
         {interpMode=RCIM_Cubic, leaveTangent=-5.2459, time=0.85, value=0.675}
         <at 0.8575 0>
        ]}
      }}
     Scale Curve=&VectorFromFloat(Value=&FloatFromCurve(
       FloatCurve=DataInterface() {propertyValues {Curve {keys [<at 0.46 0>,<at 0.51 2>,<at 0.6 1>,<at 0.7 0.55>,<at 0.8 1.2>,<at 0.9 0.7>,<at 1 1>]}}}
       Scale Curve=&RandomRangeFloat(Minimum=0.04, Maximum=0.05)
      ))
    )
    Particles.Drop Position:NiagaraPosition=$Particles.Position
   ) {enabled=false, isSetParameters=true, module=SetVariables_#4}
  ]
  rendererClasses [&NiagaraMeshRendererProperties]
  renderers [
   {rendererClass=&NiagaraMeshRendererProperties, rendererIndex=0}
   {rendererClass=&NiagaraMeshRendererProperties, rendererIndex=1}
  ]
 }
 Hero_Drips001 {
  bEnabled=false
  emitterUpdateScript [
   &EmitterState()
   &SpawnBurst_Instantaneous(Spawn Count=3, Use Spawn Probability=true) {enabled=false}
   &SpawnRate(Spawn Probability=0.8, SpawnRate=1.5, Use Spawn Probability=true)
  ]
  particleSpawnScript [
   &FN_SetSpawnGroup()
   /SetVariables(
    Particles.PendulumDrag=1
    Particles.PendulumLengthScale:NiagaraFloat=&MaskFloatBySpawnGroup(
     Masked Float=&RandomRangeFloat(Minimum=0.35, Maximum=0.5)
     Passthrough Float=&RandomRangeFloat(Minimum=0.375, Maximum=0.415)
    )
    Particles.PendulumMaskFalloff=1
    Particles.PendulumMeshWidth:NiagaraFloat=&RandomRangeFloat(Minimum=0.1, Maximum=0.125)
    Particles.PendulumRigidity=0.06
   ) {isSetParameters=true, module=SetVariables_#1}
   &InitializeParticle(
    Lifetime=&UniformRangedFloat(Minimum=2.2, Maximum=2)
    Mesh Scale=&MakeVector(X=$Particles.PendulumMeshWidth, Y=$Particles.PendulumMeshWidth, Z=$Particles.PendulumLengthScale)
    Module.Write Color=true
    Module.Write Lifetime=true
    Module.Write Mass=true
    Module.Write Material Random=true
    Module.Write Position=false
    Module.Write Scale=true
    Module.Write SpriteRotation=false
    Module.Write SpriteSize=true
   )
   &CameraOffset(Camera Offset Amount=2)
  ]
  particleUpdateScript [
   &ParticleState()
   &SkeletalMeshLocation(
    Apply To Particle Position=false
    Bone / Socket Index=$Particles.SpawnGroup
    Inherit Velocity=false
    MaskSpawnGroup=false
    Mesh Sampling Type=::NewEnumerator5
    Skeletal Mesh=$User.SkeletalMesh
    Socket Sampling Mode=::NewEnumerator5
   )
   /SetVariables(Particles.MeshOrigin:Vector3f=&Add_Vector(
     A=$Output.SkeletalMeshLocation.SampledPosition
     B=&RotateVectorByQuaternion(
      Vector To Rotate=&RandomRangeVector(Minimum=<-0.5 -0.5 -0.5>, Maximum=<0.5 0.5 0.5>)
      Quaternion=$Output.SkeletalMeshLocation.SampledRotation
     )
    )) {isSetParameters=true, module=SetVariables_#2}
   &Drag(Drag=$Particles.PendulumDrag, Use Linear Drag=true, Use Rotational Drag=false)
   &GravityForce()
   &PendulumSetup(
    Module.Enable Spring Driver=true
    Pendulum Length:NiagaraFloat=&Multiply_Float(A=100, B=$Particles.PendulumLengthScale)
    Pendulum Pivot:Vector3f=$Particles.MeshOrigin
    Pendulum Rest Axis:Vector3f=&MaskVectorBySpawnGroup(Masked Vector=<0 0 1>, Passthrough Vector=<0 0 -1>)
    "Spring Driven Constraint (Tightness)"=10
   )
   /SetVariables(Particles.MeshFacing:Vector3f=&NormalizeVector(Vector To Normalize=&Subtract_Vector(A=$Particles.MeshOrigin, B=Linked() {linkedVariable {name=Particles.Position, type=NiagaraPosition}}))) {isSetParameters=true, module=SetVariables_#3}
   &self:Pendulum_Facing(
    Bone Quat:Quat4f=$Output.SkeletalMeshLocation.SampledRotation
    FromVector:Vector3f=&MaskVectorBySpawnGroup(Masked Vector=<0 0 -1>, Passthrough Vector=<0 0 1>)
    ToVector:Vector3f=$Particles.MeshFacing
    X Vector:Vector3f=&MaskVectorBySpawnGroup(Masked Vector=<-1 0 0>, Passthrough Vector=<-1 0 0>)
    Y Vector:Vector3f=<0 0 0>
   ) {module=Pendulum_Facing}
   &SolveForcesAndVelocity(Clamp Velocity=false, Limit Acceleration=false)
   &PendulumConstraint()
   &CalculateAccurateVelocity()
   &DynamicMaterialParameters(
    Index 1 Param 3=1.0081
    Index 1 Param 4=&FloatFromCurve(FloatCurve=DataInterface() {propertyValues {Curve {keys [
         {arriveTangent=3.6645, interpMode=RCIM_Cubic, leaveTangent=3.6645, tangentMode=RCTM_User, time=0.1, value=-0.1}
         {interpMode=RCIM_Cubic, leaveTangent=2.3573, time=0.9, value=2.125}
        ]}}})
    Index 2 Param 1=HlslExpression() {hlslExpression=Particles.MeshOrigin.x}
    Index 2 Param 2=HlslExpression() {hlslExpression=Particles.MeshOrigin.y}
    Index 2 Param 3=HlslExpression() {hlslExpression=Particles.MeshOrigin.z}
    Index 2 Param 4=$Particles.PendulumMaskFalloff
    Index 3 Param 1=&MakeFloatFromVector(Vector=Linked() {linkedVariable {name=Particles.Position, type=NiagaraPosition}}, Channel=::NewEnumerator0)
    Index 3 Param 2=&MakeFloatFromVector(Vector=Linked() {linkedVariable {name=Particles.Position, type=NiagaraPosition}}, Channel=::NewEnumerator1)
    Index 3 Param 3=&MakeFloatFromVector(Vector=Linked() {linkedVariable {name=Particles.Position, type=NiagaraPosition}}, Channel=::NewEnumerator2)
    Index 3 Param 4=$Particles.PendulumRigidity
    Module.Param0WriteEnabled1=true
    Module.Param0WriteEnabled2=true
    Module.Param0WriteEnabled3=true
    Module.Param1WriteEnabled1=true
    Module.Param1WriteEnabled2=true
    Module.Param1WriteEnabled3=true
    Module.Param2WriteEnabled1=true
    Module.Param2WriteEnabled2=true
    Module.Param2WriteEnabled3=true
    Module.Param3WriteEnabled1=true
    Module.Param3WriteEnabled2=true
    Module.Param3WriteEnabled3=true
    Write Parameter Index 0=false
    Write Parameter Index 1=true
    Write Parameter Index 2=true
    Write Parameter Index 3=true
   )
   &ScaleMeshSize(Scale Factor=&MakeVector(
     X=&FloatFromCurve(FloatCurve=DataInterface() {propertyValues {Curve {keys [
          <at 0 0 RCIM_Cubic>
          {arriveTangent=2.5, interpMode=RCIM_Cubic, leaveTangent=2.5, time=0.05, value=0.9}
          {arriveTangent=-0.7353, interpMode=RCIM_Cubic, leaveTangent=-0.7353, time=0.6, value=0.65}
          {arriveTangent=-4.2857, interpMode=RCIM_Cubic, leaveTangent=-4.2857, time=0.9, value=0.375}
          {interpMode=RCIM_Cubic, leaveTangent=-0.6579, time=0.95, value=0}
         ]}}})
     Y=&FloatFromCurve(FloatCurve=DataInterface() {propertyValues {Curve {keys [
          <at 0 0 RCIM_Cubic>
          {arriveTangent=2.5, interpMode=RCIM_Cubic, leaveTangent=2.5, time=0.05, value=0.9}
          {arriveTangent=-0.7353, interpMode=RCIM_Cubic, leaveTangent=-0.7353, time=0.6, value=0.65}
          {arriveTangent=-4.2857, interpMode=RCIM_Cubic, leaveTangent=-4.2857, time=0.9, value=0.375}
          {interpMode=RCIM_Cubic, leaveTangent=-0.6579, time=0.95, value=0}
         ]}}})
     Z=&FloatFromCurve(FloatCurve=DataInterface() {propertyValues {Curve {keys [
          {arriveTangent=1.8796, interpMode=RCIM_Cubic, time=0, value=0}
          {arriveTangent=-4.7929, interpMode=RCIM_Cubic, leaveTangent=-4.7929, tangentMode=RCTM_User, time=0.625, value=0.85}
          {arriveTangent=-1.7696, interpMode=RCIM_Cubic, leaveTangent=-1.7696, tangentMode=RCTM_User, time=0.95, value=0}
         ]}}})
    ))
   /SetVariables(
    Particles.Drop Mesh Size:Vector3f=&VectorFromCurve(
     VectorCurve=DataInterface() {propertyValues {
       XCurve {keys [
         <at 0 1>
         {arriveTangent=-18.2164, interpMode=RCIM_Cubic, leaveTangent=-18.2164, tangentMode=RCTM_User, time=0.85, value=1}
         {arriveTangent=-2.286, interpMode=RCIM_Cubic, leaveTangent=-2.286, tangentMode=RCTM_User, time=0.8575, value=0}
        ]}
       YCurve {keys [
         <at 0 1>
         {arriveTangent=-18.2164, interpMode=RCIM_Cubic, leaveTangent=-18.2164, tangentMode=RCTM_User, time=0.85, value=1}
         {arriveTangent=-2.286, interpMode=RCIM_Cubic, leaveTangent=-2.286, tangentMode=RCTM_User, time=0.8575, value=0}
        ]}
       ZCurve {keys [
         <at 0.1 0.675>
         <at 0.3 0.8>
         <at 0.6 0.4>
         <at 0.8 0.8>
         {interpMode=RCIM_Cubic, leaveTangent=-5.2459, time=0.85, value=0.675}
         <at 0.8575 0>
        ]}
      }}
     Scale Curve=&VectorFromFloat(Value=&FloatFromCurve(
       FloatCurve=DataInterface() {propertyValues {Curve {keys [<at 0.46 0>,<at 0.51 2>,<at 0.6 1>,<at 0.7 0.55>,<at 0.8 1.2>,<at 0.9 0.7>,<at 1 1>]}}}
       Scale Curve=&RandomRangeFloat(Minimum=0.04, Maximum=0.05)
      ))
    )
    Particles.Drop Position:NiagaraPosition=$Particles.Position
   ) {enabled=false, isSetParameters=true, module=SetVariables_#4}
  ]
  rendererClasses [&NiagaraMeshRendererProperties]
  renderers [
   {rendererClass=&NiagaraMeshRendererProperties, rendererIndex=0}
   {rendererClass=&NiagaraMeshRendererProperties, rendererIndex=1}
  ]
 }
 Hero_Drips003 {
  bEnabled=false
  emitterUpdateScript [
   &EmitterState()
   &SpawnBurst_Instantaneous(Spawn Count=3, Use Spawn Probability=true) {enabled=false}
   &SpawnRate(SpawnRate=0.7, Use Spawn Probability=true)
  ]
  particleSpawnScript [
   &FN_SetSpawnGroup()
   /SetVariables(
    Particles.PendulumDrag=1
    Particles.PendulumLengthScale:NiagaraFloat=&MaskFloatBySpawnGroup(
     Masked Float=&RandomRangeFloat(Minimum=0.35, Maximum=0.5)
     Passthrough Float=&RandomRangeFloat(Minimum=0.375, Maximum=0.415)
    )
    Particles.PendulumMaskFalloff=1
    Particles.PendulumMeshWidth:NiagaraFloat=&RandomRangeFloat(Minimum=0.1, Maximum=0.125)
    Particles.PendulumRigidity=0.06
   ) {isSetParameters=true, module=SetVariables_#1}
   &InitializeParticle(
    Lifetime=&UniformRangedFloat(Minimum=2, Maximum=3)
    Mesh Scale=&MakeVector(X=$Particles.PendulumMeshWidth, Y=$Particles.PendulumMeshWidth, Z=$Particles.PendulumLengthScale)
    Module.Write Color=true
    Module.Write Lifetime=true
    Module.Write Mass=true
    Module.Write Material Random=true
    Module.Write Position=false
    Module.Write Scale=true
    Module.Write SpriteRotation=false
    Module.Write SpriteSize=true
   )
   &CameraOffset(Camera Offset Amount=2)
  ]
  particleUpdateScript [
   &ParticleState()
   &SkeletalMeshLocation(
    Apply To Particle Position=false
    Bone / Socket Index=$Particles.SpawnGroup
    Bone Sampling Mode=::NewEnumerator4
    Inherit Velocity=false
    MaskSpawnGroup=false
    Skeletal Mesh=$User.SkeletalMesh
   )
   /SetVariables(Particles.MeshOrigin:Vector3f=&Add_Vector(
     A=$Output.SkeletalMeshLocation.SampledPosition
     B=&RotateVectorByQuaternion(
      Vector To Rotate=&RandomRangeVector(Minimum=<-0.5 -0.5 -0.5>, Maximum=<0.5 0.5 0.5>)
      Quaternion=$Output.SkeletalMeshLocation.SampledRotation
     )
    )) {isSetParameters=true, module=SetVariables_#2}
   &Drag(Drag=$Particles.PendulumDrag, Use Linear Drag=true, Use Rotational Drag=false)
   &GravityForce()
   &PendulumSetup(
    Module.Enable Spring Driver=true
    Pendulum Length:NiagaraFloat=&Multiply_Float(A=100, B=$Particles.PendulumLengthScale)
    Pendulum Pivot:Vector3f=$Particles.MeshOrigin
    Pendulum Rest Axis:Vector3f=&MaskVectorBySpawnGroup(Masked Vector=<0 0 1>, Passthrough Vector=<0 0 -1>)
    "Spring Driven Constraint (Tightness)"=10
   )
   /SetVariables(Particles.MeshFacing:Vector3f=&NormalizeVector(Vector To Normalize=&Subtract_Vector(A=$Particles.MeshOrigin, B=Linked() {linkedVariable {name=Particles.Position, type=NiagaraPosition}}))) {isSetParameters=true, module=SetVariables_#3}
   &self:Pendulum_Facing(
    Bone Quat:Quat4f=$Output.SkeletalMeshLocation.SampledRotation
    FromVector:Vector3f=&MaskVectorBySpawnGroup(Masked Vector=<0 0 -1>, Passthrough Vector=<0 0 1>)
    ToVector:Vector3f=$Particles.MeshFacing
    X Vector:Vector3f=&MaskVectorBySpawnGroup(Masked Vector=<-1 0 0>, Passthrough Vector=<-1 0 0>)
    Y Vector:Vector3f=<0 0 0>
   ) {module=Pendulum_Facing}
   &SolveForcesAndVelocity(Clamp Velocity=false, Limit Acceleration=false)
   &PendulumConstraint()
   &CalculateAccurateVelocity()
   &DynamicMaterialParameters(
    Index 1 Param 3=1.0081
    Index 1 Param 4=&FloatFromCurve(FloatCurve=DataInterface() {propertyValues {Curve {keys [
         {arriveTangent=3.6645, interpMode=RCIM_Cubic, leaveTangent=3.6645, tangentMode=RCTM_User, time=0.1, value=-0.1}
         {interpMode=RCIM_Cubic, leaveTangent=2.3573, time=0.9, value=2.125}
        ]}}})
    Index 2 Param 1=HlslExpression() {hlslExpression=Particles.MeshOrigin.x}
    Index 2 Param 2=HlslExpression() {hlslExpression=Particles.MeshOrigin.y}
    Index 2 Param 3=HlslExpression() {hlslExpression=Particles.MeshOrigin.z}
    Index 2 Param 4=$Particles.PendulumMaskFalloff
    Index 3 Param 1=&MakeFloatFromVector(Vector=Linked() {linkedVariable {name=Particles.Position, type=NiagaraPosition}}, Channel=::NewEnumerator0)
    Index 3 Param 2=&MakeFloatFromVector(Vector=Linked() {linkedVariable {name=Particles.Position, type=NiagaraPosition}}, Channel=::NewEnumerator1)
    Index 3 Param 3=&MakeFloatFromVector(Vector=Linked() {linkedVariable {name=Particles.Position, type=NiagaraPosition}}, Channel=::NewEnumerator2)
    Index 3 Param 4=$Particles.PendulumRigidity
    Module.Param0WriteEnabled1=true
    Module.Param0WriteEnabled2=true
    Module.Param0WriteEnabled3=true
    Module.Param1WriteEnabled1=true
    Module.Param1WriteEnabled2=true
    Module.Param1WriteEnabled3=true
    Module.Param2WriteEnabled1=true
    Module.Param2WriteEnabled2=true
    Module.Param2WriteEnabled3=true
    Module.Param3WriteEnabled1=true
    Module.Param3WriteEnabled2=true
    Module.Param3WriteEnabled3=true
    Write Parameter Index 0=false
    Write Parameter Index 1=true
    Write Parameter Index 2=true
    Write Parameter Index 3=true
   )
   &ScaleMeshSize(Scale Factor=&MakeVector(
     X=&FloatFromCurve(FloatCurve=DataInterface() {propertyValues {Curve {keys [
          <at 0 0 RCIM_Cubic>
          {arriveTangent=2.5, interpMode=RCIM_Cubic, leaveTangent=2.5, time=0.05, value=0.9}
          {arriveTangent=-0.7353, interpMode=RCIM_Cubic, leaveTangent=-0.7353, time=0.6, value=0.65}
          {arriveTangent=-4.2857, interpMode=RCIM_Cubic, leaveTangent=-4.2857, time=0.9, value=0.375}
          {interpMode=RCIM_Cubic, leaveTangent=-0.6579, time=0.95, value=0}
         ]}}})
     Y=&FloatFromCurve(FloatCurve=DataInterface() {propertyValues {Curve {keys [
          <at 0 0 RCIM_Cubic>
          {arriveTangent=2.5, interpMode=RCIM_Cubic, leaveTangent=2.5, time=0.05, value=0.9}
          {arriveTangent=-0.7353, interpMode=RCIM_Cubic, leaveTangent=-0.7353, time=0.6, value=0.65}
          {arriveTangent=-4.2857, interpMode=RCIM_Cubic, leaveTangent=-4.2857, time=0.9, value=0.375}
          {interpMode=RCIM_Cubic, leaveTangent=-0.6579, time=0.95, value=0}
         ]}}})
     Z=&FloatFromCurve(FloatCurve=DataInterface() {propertyValues {Curve {keys [
          {arriveTangent=1.8796, interpMode=RCIM_Cubic, time=0, value=0}
          {arriveTangent=-4.7929, interpMode=RCIM_Cubic, leaveTangent=-4.7929, tangentMode=RCTM_User, time=0.625, value=0.85}
          {arriveTangent=-1.7696, interpMode=RCIM_Cubic, leaveTangent=-1.7696, tangentMode=RCTM_User, time=0.95, value=0}
         ]}}})
    ))
   /SetVariables(
    Particles.Drop Mesh Size:Vector3f=&VectorFromCurve(
     VectorCurve=DataInterface() {propertyValues {
       XCurve {keys [
         <at 0 1>
         {arriveTangent=-18.2164, interpMode=RCIM_Cubic, leaveTangent=-18.2164, tangentMode=RCTM_User, time=0.85, value=1}
         {arriveTangent=-2.286, interpMode=RCIM_Cubic, leaveTangent=-2.286, tangentMode=RCTM_User, time=0.8575, value=0}
        ]}
       YCurve {keys [
         <at 0 1>
         {arriveTangent=-18.2164, interpMode=RCIM_Cubic, leaveTangent=-18.2164, tangentMode=RCTM_User, time=0.85, value=1}
         {arriveTangent=-2.286, interpMode=RCIM_Cubic, leaveTangent=-2.286, tangentMode=RCTM_User, time=0.8575, value=0}
        ]}
       ZCurve {keys [
         <at 0.1 0.675>
         <at 0.3 0.8>
         <at 0.6 0.4>
         <at 0.8 0.8>
         {interpMode=RCIM_Cubic, leaveTangent=-5.2459, time=0.85, value=0.675}
         <at 0.8575 0>
        ]}
      }}
     Scale Curve=&VectorFromFloat(Value=&FloatFromCurve(
       FloatCurve=DataInterface() {propertyValues {Curve {keys [<at 0.46 0>,<at 0.51 2>,<at 0.6 1>,<at 0.7 0.55>,<at 0.8 1.2>,<at 0.9 0.7>,<at 1 1>]}}}
       Scale Curve=&RandomRangeFloat(Minimum=0.04, Maximum=0.05)
      ))
    )
    Particles.Drop Position:NiagaraPosition=$Particles.Position
   ) {enabled=false, isSetParameters=true, module=SetVariables_#4}
  ]
  rendererClasses [&NiagaraMeshRendererProperties]
  renderers [
   {rendererClass=&NiagaraMeshRendererProperties, rendererIndex=0}
   {rendererClass=&NiagaraMeshRendererProperties, rendererIndex=1}
  ]
 }
 Ribbon_BoneFlow {
  bEnabled=false
  emitterUpdateScript [
   &EmitterState(Life Cycle Mode=::NewEnumerator1, Loop Duration=1, UseLoopDelay=false)
   &SpawnRate(SpawnRate=&RandomRangeFloat(Minimum=20, Maximum=30), Use Spawn Probability=false)
  ]
  particleSpawnScript [
   &FN_SetSpawnGroup(Inclusivity=true, Normalize Spawn Group=true, Particles Per Group=10, Pattern=::NewEnumerator0, Set Ribbon ID and Ribbon Link Order=true)
   &InitializeRibbon(
    Lifetime Max=1.5
    Lifetime Min=0.8
    Lifetime Mode:ENiagara_LifetimeMode=::NewEnumerator1
    Ribbon Width=&RandomRangeFloat(Minimum=2, Maximum=1.2)
    UsePositionOffset=false
    Write Material Random=true
   )
  ]
  particleUpdateScript [
   &ParticleState()
   /Game/Effects/Niagara/Modules/BoneFlow/FN_BoneFlow_v4/FN_BoneFlow(
    Angle:NiagaraFloat=&Multiply_Float(A=$Particles.Normalized Spawn Group, B=&RandomRangeFloat(Minimum=-0.6, Maximum=0.6))
    BoneID:NiagaraInt32=&ModuloInt(A=$Particles.SpawnGroup, B=1)
    Height:NiagaraFloat=&Multiply_Float(
     A=$Particles.Normalized Spawn Group
     B=&FloatFromCurve(
      FloatCurve=DataInterface() {propertyValues {Curve {keys [
          {arriveTangent=0.2329, interpMode=RCIM_Cubic, leaveTangent=0.2329, tangentMode=RCTM_User, time=0, value=1}
          {interpMode=RCIM_Cubic, tangentMode=RCTM_User, time=1, value=1.1}
         ]}}}
      Scale Curve=0.5
     )
    )
    Radius:NiagaraFloat=&Multiply_Float(
     A=$Particles.Normalized Spawn Group
     B=&FloatFromCurve(
      FloatCurve=DataInterface() {propertyValues {Curve {keys [
          {arriveTangent=4.8142, interpMode=RCIM_Cubic, leaveTangent=4.8142, tangentMode=RCTM_User, time=0, value=10}
          {arriveTangent=-3.363, interpMode=RCIM_Cubic, leaveTangent=-3.363, tangentMode=RCTM_User, time=0.5, value=10}
          {arriveTangent=-23.2104, interpMode=RCIM_Cubic, leaveTangent=-23.2104, tangentMode=RCTM_User, time=1, value=1}
         ]}}}
      CurveIndex=$Particles.Normalized Spawn Group
      Scale Curve=&RandomRangeFloat(Minimum=2, Maximum=1)
     )
    )
    SkeletalMesh=DataInterface() {propertyValues {
      DefaultMesh=null
      ExcludeBoneName=root
      OptionalComponentClass=null
      PreviewMesh=&Psyche_Male_Axe
      SkinningMode=SkinOnTheFly
      SoftSourceActor=null
      SourceMode=Default
      UvSetIndex=0
      WholeMeshLOD=-1
      bExcludeBone=false
      bReadDeformedGeometry=true
      bRequireCurrentFrameData=true
      ComponentTags []
      FilteredBones []
      FilteredSockets [BoneFlow01,BoneFlow02,BoneFlow03]
      MeshUserParameter {parameter {name=None, typeDefHandle {registeredTypeIndex=1}, varData []}}
      SamplingRegions []
     }}
    Use Sockets Instead=true
    Write Position Attribute=true
   ) {module=FN_BoneFlow001}
   &ScaleRibbonWidth(Ribbon Width Scale=&FloatFromCurve(
     FloatCurve=DataInterface() {propertyValues {Curve {keys [
         {interpMode=RCIM_Cubic, tangentMode=RCTM_User, time=0, value=0}
         {interpMode=RCIM_Cubic, tangentMode=RCTM_User, time=0.2, value=2}
         {arriveTangent=-8.5417, interpMode=RCIM_Cubic, leaveTangent=-8.5417, tangentMode=RCTM_User, time=0.6, value=0.575}
         {arriveTangent=-2.6171, interpMode=RCIM_Cubic, leaveTangent=-2.6171, tangentMode=RCTM_User, time=0.8, value=1}
         {arriveTangent=-0.2025, interpMode=RCIM_Cubic, leaveTangent=-0.2025, tangentMode=RCTM_User, time=1, value=0}
        ]}}}
     CurveIndex=$Particles.Normalized Spawn Group
     Scale Curve=&FloatFromCurve(
      FloatCurve=DataInterface() {propertyValues {Curve {keys [
          {arriveTangent=2.8989, interpMode=RCIM_Cubic, leaveTangent=2.8989, tangentMode=RCTM_User, time=0.01, value=0}
          {interpMode=RCIM_Cubic, tangentMode=RCTM_User, time=0.4, value=0.95}
          {interpMode=RCIM_Cubic, tangentMode=RCTM_User, time=0.6, value=1}
          {arriveTangent=-1.5059, interpMode=RCIM_Cubic, leaveTangent=-1.5059, tangentMode=RCTM_User, time=0.99, value=0}
         ]}}}
      CurveIndex=$Particles.Age
     )
    ))
   /SetVariables(Particles.RibbonFacing=$Output.FN_BoneFlow001.SplineBinormal) {isSetParameters=true, module=SetVariables_#5}
   &DynamicMaterialParameters(
    Index 0 Param 1=&FloatFromCurve(FloatCurve=DataInterface() {propertyValues {Curve {keys [<at 0 0 RCIM_Cubic>,<at 0.5 1>,<at 1 0 RCIM_Cubic>]}}})
    Module.Param0WriteEnabled=true
    Module.Param1WriteEnabled=true
    Module.Param2WriteEnabled=false
    Module.Param3WriteEnabled=false
   )
   &SolveForcesAndVelocity(Clamp Velocity=false, Limit Acceleration=false)
  ]
  rendererClasses [&NiagaraRibbonRendererProperties]
  renderers [{rendererClass=&NiagaraRibbonRendererProperties}]
 }
 Ribbon_BoneFlow_Continuous {
  emitterUpdateScript [
   &EmitterState(Life Cycle Mode=::NewEnumerator1, Loop Duration=1, UseLoopDelay=false)
   &SpawnRate(SpawnRate=&RandomRangeFloat(Minimum=30, Maximum=60), Use Spawn Probability=false)
  ]
  particleSpawnScript [
   &FN_SetSpawnGroup(Inclusivity=true, Normalize Spawn Group=true, Particles Per Group=10, Pattern=::NewEnumerator0, Set Ribbon ID and Ribbon Link Order=true)
   &InitializeRibbon(
    Lifetime Max=1.5
    Lifetime Min=0.8
    Lifetime Mode:ENiagara_LifetimeMode=::NewEnumerator1
    Ribbon Width=&RandomRangeFloat(Minimum=2, Maximum=1.5)
    UsePositionOffset=false
    Write Material Random=true
   )
  ]
  particleUpdateScript [
   &ParticleState()
   &FN_BoneFlow(
    Angle:NiagaraFloat=&Multiply_Float(A=$Particles.Normalized Spawn Group, B=&RandomRangeFloat(Minimum=-0.6, Maximum=0.6))
    BoneID:NiagaraInt32=&ModuloInt(A=$Particles.SpawnGroup, B=1)
    Height:NiagaraFloat=&Multiply_Float(
     A=$Particles.Normalized Spawn Group
     B=&FloatFromCurve(FloatCurve=DataInterface() {propertyValues {Curve {keys [
          {arriveTangent=0.2329, interpMode=RCIM_Cubic, leaveTangent=0.2329, tangentMode=RCTM_User, time=0, value=1}
          {interpMode=RCIM_Cubic, tangentMode=RCTM_User, time=1, value=1.1}
         ]}}})
    )
    Radius:NiagaraFloat=&Multiply_Float(
     A=$Particles.Normalized Spawn Group
     B=&FloatFromCurve(
      FloatCurve=DataInterface() {propertyValues {Curve {keys [
          {arriveTangent=4.8142, interpMode=RCIM_Cubic, leaveTangent=4.8142, tangentMode=RCTM_User, time=-0.1, value=400}
          {interpMode=RCIM_Cubic, tangentMode=RCTM_User, time=0.05, value=30}
          {arriveTangent=-26.8012, interpMode=RCIM_Cubic, leaveTangent=-26.8012, tangentMode=RCTM_User, time=0.1, value=28.75}
          {arriveTangent=-23.2104, interpMode=RCIM_Cubic, leaveTangent=-23.2104, tangentMode=RCTM_User, time=1, value=0}
         ]}}}
      CurveIndex=$Particles.Normalized Spawn Group
      Scale Curve=&Multiply_Float(A=&RandomRangeFloat(Minimum=-1, Maximum=1), B=0.3)
     )
    )
    SkeletalMesh=DataInterface() {propertyValues {
      DefaultMesh=null
      ExcludeBoneName=root
      OptionalComponentClass=null
      PreviewMesh=&Psyche_Male_Axe
      SkinningMode=SkinOnTheFly
      SoftSourceActor=null
      SourceMode=Default
      UvSetIndex=0
      WholeMeshLOD=-1
      bExcludeBone=false
      bReadDeformedGeometry=true
      bRequireCurrentFrameData=true
      ComponentTags []
      FilteredBones []
      FilteredSockets [BoneFlow01,BoneFlow02,BoneFlow03]
      MeshUserParameter {parameter {name=None, typeDefHandle {registeredTypeIndex=1}, varData []}}
      SamplingRegions []
     }}
    Stabalize Twisting=true
    Use Sockets Instead=true
    Write Position Attribute=true
   ) {module=FN_BoneFlow001}
   &ScaleRibbonWidth(Ribbon Width Scale=&FloatFromCurve(
     FloatCurve=DataInterface() {propertyValues {Curve {keys [
         {interpMode=RCIM_Cubic, tangentMode=RCTM_User, time=0, value=0}
         {interpMode=RCIM_Cubic, tangentMode=RCTM_User, time=0.25, value=1.7}
         {arriveTangent=-8.5417, interpMode=RCIM_Cubic, leaveTangent=-8.5417, tangentMode=RCTM_User, time=0.6, value=0.575}
         {arriveTangent=-2.6171, interpMode=RCIM_Cubic, leaveTangent=-2.6171, tangentMode=RCTM_User, time=0.8, value=1.3}
         <at 1 0>
        ]}}}
     CurveIndex=$Particles.Normalized Spawn Group
     Scale Curve=&FloatFromCurve(
      FloatCurve=DataInterface() {propertyValues {Curve {keys [
          {arriveTangent=2.8989, interpMode=RCIM_Cubic, leaveTangent=2.8989, tangentMode=RCTM_User, time=0.01, value=0}
          {interpMode=RCIM_Cubic, tangentMode=RCTM_User, time=0.5, value=1}
          {arriveTangent=-1.5059, interpMode=RCIM_Cubic, leaveTangent=-1.5059, tangentMode=RCTM_User, time=0.99, value=0}
         ]}}}
      CurveIndex=$Particles.Age
      Scale Curve=0.8
     )
    ))
   /SetVariables(Particles.RibbonFacing=$Output.FN_BoneFlow001.SplineBinormal) {isSetParameters=true, module=SetVariables_#5}
   &DynamicMaterialParameters(
    Index 0 Param 1=&FloatFromCurve(FloatCurve=DataInterface() {propertyValues {Curve {keys [<at 0 0 RCIM_Cubic>,<at 0.5 1>,<at 1 0 RCIM_Cubic>]}}})
    Module.Param0WriteEnabled=true
    Module.Param0WriteEnabled1=true
    Module.Param0WriteEnabled2=true
    Module.Param0WriteEnabled3=true
    Module.Param1WriteEnabled=true
    Module.Param1WriteEnabled1=true
    Module.Param1WriteEnabled2=true
    Module.Param1WriteEnabled3=true
    Module.Param2WriteEnabled=true
    Module.Param2WriteEnabled1=true
    Module.Param2WriteEnabled2=true
    Module.Param2WriteEnabled3=true
    Module.Param3WriteEnabled=true
    Module.Param3WriteEnabled1=true
    Module.Param3WriteEnabled2=true
    Module.Param3WriteEnabled3=true
    Write Parameter Index 1=true
    Write Parameter Index 2=true
    Write Parameter Index 3=true
   )
   &SolveForcesAndVelocity(Clamp Velocity=false, Limit Acceleration=false)
  ]
  rendererClasses [&NiagaraRibbonRendererProperties]
  renderers [{rendererClass=&NiagaraRibbonRendererProperties}]
 }
 Ribbon_BoneFlow_Instantaneous {
  emitterUpdateScript [
   &EmitterState(Life Cycle Mode=::NewEnumerator1, Loop Behavior=::NewEnumerator1, Loop Duration=0.1, UseLoopDelay=false)
   &SpawnRate(SpawnRate=&RandomRangeFloat(Minimum=20, Maximum=30), Use Spawn Probability=false) {enabled=false}
   &SpawnBurst_Instantaneous(Spawn Count=20, Use Spawn Probability=false, UseLoopCountLimit=false)
  ]
  particleSpawnScript [
   &FN_SetSpawnGroup(Inclusivity=true, Normalize Spawn Group=true, Particles Per Group=10, Pattern=::NewEnumerator0, Set Ribbon ID and Ribbon Link Order=true)
   &InitializeRibbon(
    Lifetime Max=1.5
    Lifetime Min=0.8
    Lifetime Mode:ENiagara_LifetimeMode=::NewEnumerator1
    Ribbon Width=&RandomRangeFloat(Minimum=2, Maximum=1.5)
    UsePositionOffset=false
    Write Material Random=true
   )
  ]
  particleUpdateScript [
   &ParticleState()
   &FN_BoneFlow(
    Angle:NiagaraFloat=&Multiply_Float(A=$Particles.Normalized Spawn Group, B=&RandomRangeFloat(Minimum=-0.6, Maximum=0.6))
    BoneID:NiagaraInt32=&ModuloInt(A=$Particles.SpawnGroup, B=1)
    Height:NiagaraFloat=&Multiply_Float(
     A=$Particles.Normalized Spawn Group
     B=&FloatFromCurve(FloatCurve=DataInterface() {propertyValues {Curve {keys [
          {arriveTangent=0.2329, interpMode=RCIM_Cubic, leaveTangent=0.2329, tangentMode=RCTM_User, time=0, value=1}
          {interpMode=RCIM_Cubic, tangentMode=RCTM_User, time=1, value=1.1}
         ]}}})
    )
    Radius:NiagaraFloat=&Multiply_Float(
     A=$Particles.Normalized Spawn Group
     B=&FloatFromCurve(
      FloatCurve=DataInterface() {propertyValues {Curve {keys [
          {arriveTangent=4.8142, interpMode=RCIM_Cubic, leaveTangent=4.8142, tangentMode=RCTM_User, time=0, value=0}
          {arriveTangent=-26.8012, interpMode=RCIM_Cubic, leaveTangent=-26.8012, tangentMode=RCTM_User, time=0.2, value=10}
          {arriveTangent=-23.2104, interpMode=RCIM_Cubic, leaveTangent=-23.2104, tangentMode=RCTM_User, time=1, value=0}
         ]}}}
      CurveIndex=$Particles.Normalized Spawn Group
      Scale Curve=&RandomRangeFloat(Minimum=1, Maximum=-1)
     )
    )
    SkeletalMesh=DataInterface() {propertyValues {
      DefaultMesh=null
      ExcludeBoneName=root
      OptionalComponentClass=null
      PreviewMesh=&Psyche_Male_Axe
      SkinningMode=SkinOnTheFly
      SoftSourceActor=null
      SourceMode=Default
      UvSetIndex=0
      WholeMeshLOD=-1
      bExcludeBone=false
      bReadDeformedGeometry=true
      bRequireCurrentFrameData=true
      ComponentTags []
      FilteredBones []
      FilteredSockets [BoneFlow01,BoneFlow02,BoneFlow03]
      MeshUserParameter {parameter {name=None, typeDefHandle {registeredTypeIndex=1}, varData []}}
      SamplingRegions []
     }}
    Stabalize Twisting=true
    Use Sockets Instead=true
    Write Position Attribute=true
   ) {module=FN_BoneFlow001}
   &ScaleRibbonWidth(Ribbon Width Scale=&FloatFromCurve(
     FloatCurve=DataInterface() {propertyValues {Curve {keys [
         {interpMode=RCIM_Cubic, tangentMode=RCTM_User, time=0, value=1}
         {interpMode=RCIM_Cubic, tangentMode=RCTM_User, time=0.3, value=2}
         {arriveTangent=-8.5417, interpMode=RCIM_Cubic, leaveTangent=-8.5417, tangentMode=RCTM_User, time=0.6, value=0.575}
         {arriveTangent=-2.6171, interpMode=RCIM_Cubic, leaveTangent=-2.6171, tangentMode=RCTM_User, time=0.8, value=1}
         <at 1 0>
        ]}}}
     CurveIndex=$Particles.Normalized Spawn Group
     Scale Curve=&FloatFromCurve(
      FloatCurve=DataInterface() {propertyValues {Curve {keys [
          {arriveTangent=2.8989, interpMode=RCIM_Cubic, leaveTangent=2.8989, tangentMode=RCTM_User, time=0.01, value=1}
          {interpMode=RCIM_Cubic, tangentMode=RCTM_User, time=0.4, value=1}
          {interpMode=RCIM_Cubic, tangentMode=RCTM_User, time=0.6, value=1}
          {arriveTangent=-1.5059, interpMode=RCIM_Cubic, leaveTangent=-1.5059, tangentMode=RCTM_User, time=1, value=0}
         ]}}}
      CurveIndex=$Particles.Age
      Scale Curve=0.5
     )
    ))
   /SetVariables(Particles.RibbonFacing=$Output.FN_BoneFlow001.SplineBinormal) {isSetParameters=true, module=SetVariables_#5}
   &DynamicMaterialParameters(
    Index 0 Param 1=&FloatFromCurve(FloatCurve=DataInterface() {propertyValues {Curve {keys [<at 0 0 RCIM_Cubic>,<at 0.5 1>,<at 1 0 RCIM_Cubic>]}}})
    Module.Param0WriteEnabled=true
    Module.Param0WriteEnabled1=true
    Module.Param0WriteEnabled2=true
    Module.Param0WriteEnabled3=true
    Module.Param1WriteEnabled=true
    Module.Param1WriteEnabled1=true
    Module.Param1WriteEnabled2=true
    Module.Param1WriteEnabled3=true
    Module.Param2WriteEnabled=true
    Module.Param2WriteEnabled1=true
    Module.Param2WriteEnabled2=true
    Module.Param2WriteEnabled3=true
    Module.Param3WriteEnabled=true
    Module.Param3WriteEnabled1=true
    Module.Param3WriteEnabled2=true
    Module.Param3WriteEnabled3=true
    Write Parameter Index 1=true
    Write Parameter Index 2=true
    Write Parameter Index 3=true
   )
   &SolveForcesAndVelocity(Clamp Velocity=false, Limit Acceleration=false)
   &ScaleRibbonWidth(Ribbon Width Scale=&FloatFromCurve(
     FloatCurve=DataInterface() {propertyValues {Curve {keys [
         {interpMode=RCIM_Cubic, tangentMode=RCTM_User, time=0, value=0.2}
         {interpMode=RCIM_Cubic, tangentMode=RCTM_User, time=0.5, value=1}
         {interpMode=RCIM_Cubic, tangentMode=RCTM_User, time=1, value=0.2}
        ]}}}
     CurveIndex=$Particles.RibbonLinkOrder
    )) {enabled=false, module=ScaleRibbonWidth001}
  ]
  rendererClasses [&NiagaraRibbonRendererProperties]
  renderers [{rendererClass=&NiagaraRibbonRendererProperties}]
 }
}
summary {
 systemName=NS_Pickaxe_Psyche_Idle
 emitters [
  {bEnabled=false, emitterName=Hero_Drips, rendererClasses [&NiagaraMeshRendererProperties]}
  {bEnabled=false, emitterName=Hero_Drips001, rendererClasses [&NiagaraMeshRendererProperties]}
  {bEnabled=false, emitterName=Hero_Drips003, rendererClasses [&NiagaraMeshRendererProperties]}
  {emitterName=GoopDrop, rendererClasses [&NiagaraMeshRendererProperties]}
  {bEnabled=false, emitterName=Ribbon_BoneFlow, rendererClasses [&NiagaraRibbonRendererProperties]}
  {emitterName=Ribbon_BoneFlow_Continuous, rendererClasses [&NiagaraRibbonRendererProperties]}
  {emitterName=Ribbon_BoneFlow_Instantaneous, rendererClasses [&NiagaraRibbonRendererProperties]}
 ]
 userVariables [
  User.bIsFrontEnd {type=NiagaraBool}
  User.bIsFrontEndPreview {type=NiagaraBool}
  User.BladeSwing {type=NiagaraBool}
  User.SkeletalMesh {
   type=NiagaraDataInterfaceSkeletalMesh
   defaultValue {
    struct=NiagaraExt_VariableValue_DataInterface
    value {dataInterface=&self:NiagaraDataInterfaceSkeletalMesh_0}
   }
  }
  User.SlurpColorAlpha {type=NiagaraFloat, defaultValue=0, description=""}
 ]
}
