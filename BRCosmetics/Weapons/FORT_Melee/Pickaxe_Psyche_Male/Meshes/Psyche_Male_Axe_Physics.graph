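# Physics asset for the Psyche Male pickaxe mesh, defining capsule bodies and joints for its dangling cable and glob bones.
# Cable_1, globMain_1 and globMain_2 constrain to the two-handed handle.
# Thirteen identical two-link glob chains (A–M) branch from globMain_1, with looser swing at the outer link and no twist.
# The other cable bones chain off globMain_2 in reverse order, cable_3 then cable_2, rather than continuing from cable_1.
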PhysicsAsset /BRCosmetics/Weapons/FORT_Melee/Pickaxe_Psyche_Male/Meshes/Psyche_Male_Axe_Physics
bodies {
 Melee_TwoHanded_Handle [{shapeType=Capsule, length=2, radius=2, shapeName=Melee_TwoHanded_Handle_capsule}]
 dyn_cable_1 [{shapeType=Capsule, length=2, radius=2, shapeName=dyn_cable_1_capsule, center=<1.8 0 0>, rotation=<rot -90 0 0>}]
 dyn_cable_2 [{shapeType=Capsule, length=2, radius=2, shapeName=dyn_cable_2_capsule, center=<1.5102 0 0>, rotation=<rot 90 0 0>}]
 dyn_cable_3 [{shapeType=Capsule, length=2, radius=2, shapeName=dyn_cable_3_capsule, rotation=<rot 90 0 0>}]
 dyn_globA_1 [{shapeType=Capsule, length=2, radius=2.5, shapeName=dyn_globA_1_capsule, center=<8.4139 -0.0002 0>}]
 dyn_globA_2 [{shapeType=Capsule, length=2, radius=1.75, shapeName=dyn_globA_2_capsule, center=<2.2748 0.0001 0.0001>, rotation=<rot 90 0 0>}]
 dyn_globB_1 [{shapeType=Capsule, length=2, radius=2.5, shapeName=dyn_globB_1_capsule, center=<8.4139 -0.0002 0>}]
 dyn_globB_2 [{shapeType=Capsule, length=2, radius=1.75, shapeName=dyn_globB_2_capsule, center=<2.2748 0.0001 0.0001>, rotation=<rot 90 0 0>}]
 dyn_globC_1 [{shapeType=Capsule, length=2, radius=2.5, shapeName=dyn_globC_1_capsule, center=<8.4139 -0.0002 0>}]
 dyn_globC_2 [{shapeType=Capsule, length=2, radius=1.75, shapeName=dyn_globC_2_capsule, center=<2.2748 0.0001 0.0001>, rotation=<rot 90 0 0>}]
 dyn_globD_1 [{shapeType=Capsule, length=2, radius=2.5, shapeName=dyn_globD_1_capsule, center=<8.4139 -0.0002 0>}]
 dyn_globD_2 [{shapeType=Capsule, length=2, radius=1.75, shapeName=dyn_globD_2_capsule, center=<2.2748 0.0001 0.0001>, rotation=<rot 90 0 0>}]
 dyn_globE_1 [{shapeType=Capsule, length=2, radius=2.5, shapeName=dyn_globE_1_capsule, center=<8.4139 -0.0002 0>}]
 dyn_globE_2 [{shapeType=Capsule, length=2, radius=1.75, shapeName=dyn_globE_2_capsule, center=<2.2748 0.0001 0.0001>, rotation=<rot 90 0 0>}]
 dyn_globF_1 [{shapeType=Capsule, length=2, radius=2.5, shapeName=dyn_globF_1_capsule, center=<8.4139 -0.0002 0>}]
 dyn_globF_2 [{shapeType=Capsule, length=2, radius=1.75, shapeName=dyn_globF_2_capsule, center=<2.2748 0.0001 0.0001>, rotation=<rot 90 0 0>}]
 dyn_globG_1 [{shapeType=Capsule, length=2, radius=2.5, shapeName=dyn_globG_1_capsule, center=<8.4139 -0.0002 0>}]
 dyn_globG_2 [{shapeType=Capsule, length=2, radius=1.75, shapeName=dyn_globG_2_capsule, center=<2.2748 0.0001 0.0001>, rotation=<rot 90 0 0>}]
 dyn_globH_1 [{shapeType=Capsule, length=2, radius=2.5, shapeName=dyn_globH_1_capsule, center=<8.4139 -0.0002 0>}]
 dyn_globH_2 [{shapeType=Capsule, length=2, radius=1.75, shapeName=dyn_globH_2_capsule, center=<2.2748 0.0001 0.0001>, rotation=<rot 90 0 0>}]
 dyn_globI_1 [{shapeType=Capsule, length=2, radius=2.5, shapeName=dyn_globI_1_capsule, center=<8.4139 -0.0002 0>}]
 dyn_globI_2 [{shapeType=Capsule, length=2, radius=1.75, shapeName=dyn_globI_2_capsule, center=<2.2748 0.0001 0.0001>, rotation=<rot 90 0 0>}]
 dyn_globJ_1 [{shapeType=Capsule, length=2, radius=2.5, shapeName=dyn_globJ_1_capsule, center=<8.4139 -0.0002 0>}]
 dyn_globJ_2 [{shapeType=Capsule, length=2, radius=1.75, shapeName=dyn_globJ_2_capsule, center=<2.2748 0.0001 0.0001>, rotation=<rot 90 0 0>}]
 dyn_globK_1 [{shapeType=Capsule, length=2, radius=2.5, shapeName=dyn_globK_1_capsule, center=<8.4139 -0.0002 0>}]
 dyn_globK_2 [{shapeType=Capsule, length=2, radius=1.75, shapeName=dyn_globK_2_capsule, center=<2.2748 0.0001 0.0001>, rotation=<rot 90 0 0>}]
 dyn_globL_1 [{shapeType=Capsule, length=2, radius=2.5, shapeName=dyn_globL_1_capsule, center=<8.4139 -0.0002 0>}]
 dyn_globL_2 [{shapeType=Capsule, length=2, radius=1.75, shapeName=dyn_globL_2_capsule, center=<2.2748 0.0001 0.0001>, rotation=<rot 90 0 0>}]
 dyn_globM_1 [{shapeType=Capsule, length=2, radius=2.5, shapeName=dyn_globM_1_capsule, center=<8.4139 -0.0002 0>}]
 dyn_globM_2 [{shapeType=Capsule, length=2, radius=1.75, shapeName=dyn_globM_2_capsule, center=<2.2748 0.0001 0.0001>, rotation=<rot 90 0 0>}]
 dyn_globMain_1 [{shapeType=Capsule, length=0.505, radius=0.505, shapeName=dyn_globMain_1_capsule, rotation=<rot 0 -90 0>}]
 dyn_globMain_2 [{shapeType=Capsule, length=2, radius=6.7089, shapeName=dyn_globMain_2_capsule, center=<-10.2719 -3.1266 3.5533>}]
}
constraints [
 {bone1Name=dyn_cable_1, bone2Name=Melee_TwoHanded_Handle, swing1LimitDegrees=25, swing2LimitDegrees=25, twistLimitDegrees=0}
 {bone1Name=dyn_globMain_1, bone2Name=Melee_TwoHanded_Handle, swing1LimitDegrees=45, swing2LimitDegrees=45, twistLimitDegrees=45}
 {bone1Name=dyn_globA_1, bone2Name=dyn_globMain_1, swing1LimitDegrees=17, swing2LimitDegrees=17, twistLimitDegrees=0}
 {bone1Name=dyn_globA_2, bone2Name=dyn_globA_1, swing1LimitDegrees=30, swing2LimitDegrees=30, twistLimitDegrees=0}
 {bone1Name=dyn_globB_1, bone2Name=dyn_globMain_1, swing1LimitDegrees=17, swing2LimitDegrees=17, twistLimitDegrees=0}
 {bone1Name=dyn_globB_2, bone2Name=dyn_globB_1, swing1LimitDegrees=30, swing2LimitDegrees=30, twistLimitDegrees=0}
 {bone1Name=dyn_globC_1, bone2Name=dyn_globMain_1, swing1LimitDegrees=17, swing2LimitDegrees=17, twistLimitDegrees=0}
 {bone1Name=dyn_globC_2, bone2Name=dyn_globC_1, swing1LimitDegrees=30, swing2LimitDegrees=30, twistLimitDegrees=0}
 {bone1Name=dyn_globD_1, bone2Name=dyn_globMain_1, swing1LimitDegrees=17, swing2LimitDegrees=17, twistLimitDegrees=0}
 {bone1Name=dyn_globD_2, bone2Name=dyn_globD_1, swing1LimitDegrees=30, swing2LimitDegrees=30, twistLimitDegrees=0}
 {bone1Name=dyn_globE_1, bone2Name=dyn_globMain_1, swing1LimitDegrees=17, swing2LimitDegrees=17, twistLimitDegrees=0}
 {bone1Name=dyn_globE_2, bone2Name=dyn_globE_1, swing1LimitDegrees=30, swing2LimitDegrees=30, twistLimitDegrees=0}
 {bone1Name=dyn_globF_1, bone2Name=dyn_globMain_1, swing1LimitDegrees=17, swing2LimitDegrees=17, twistLimitDegrees=0}
 {bone1Name=dyn_globF_2, bone2Name=dyn_globF_1, swing1LimitDegrees=30, swing2LimitDegrees=30, twistLimitDegrees=0}
 {bone1Name=dyn_globG_1, bone2Name=dyn_globMain_1, swing1LimitDegrees=17, swing2LimitDegrees=17, twistLimitDegrees=0}
 {bone1Name=dyn_globG_2, bone2Name=dyn_globG_1, swing1LimitDegrees=30, swing2LimitDegrees=30, twistLimitDegrees=0}
 {bone1Name=dyn_globH_1, bone2Name=dyn_globMain_1, swing1LimitDegrees=17, swing2LimitDegrees=17, twistLimitDegrees=0}
 {bone1Name=dyn_globH_2, bone2Name=dyn_globH_1, swing1LimitDegrees=30, swing2LimitDegrees=30, twistLimitDegrees=0}
 {bone1Name=dyn_globI_1, bone2Name=dyn_globMain_1, swing1LimitDegrees=17, swing2LimitDegrees=17, twistLimitDegrees=0}
 {bone1Name=dyn_globI_2, bone2Name=dyn_globI_1, swing1LimitDegrees=30, swing2LimitDegrees=30, twistLimitDegrees=0}
 {bone1Name=dyn_globJ_1, bone2Name=dyn_globMain_1, swing1LimitDegrees=17, swing2LimitDegrees=17, twistLimitDegrees=0}
 {bone1Name=dyn_globJ_2, bone2Name=dyn_globJ_1, swing1LimitDegrees=30, swing2LimitDegrees=30, twistLimitDegrees=0}
 {bone1Name=dyn_globK_1, bone2Name=dyn_globMain_1, swing1LimitDegrees=17, swing2LimitDegrees=17, twistLimitDegrees=0}
 {bone1Name=dyn_globK_2, bone2Name=dyn_globK_1, swing1LimitDegrees=30, swing2LimitDegrees=30, twistLimitDegrees=0}
 {bone1Name=dyn_globL_1, bone2Name=dyn_globMain_1, swing1LimitDegrees=17, swing2LimitDegrees=17, twistLimitDegrees=0}
 {bone1Name=dyn_globL_2, bone2Name=dyn_globL_1, swing1LimitDegrees=30, swing2LimitDegrees=30, twistLimitDegrees=0}
 {bone1Name=dyn_globM_1, bone2Name=dyn_globMain_1, swing1LimitDegrees=17, swing2LimitDegrees=17, twistLimitDegrees=0}
 {bone1Name=dyn_globM_2, bone2Name=dyn_globM_1, swing1LimitDegrees=30, swing2LimitDegrees=30, twistLimitDegrees=0}
 {bone1Name=dyn_globMain_2, bone2Name=Melee_TwoHanded_Handle, swing1LimitDegrees=15, swing2LimitDegrees=15, twistLimitDegrees=0}
 {bone1Name=dyn_cable_3, bone2Name=dyn_globMain_2, swing1LimitDegrees=15, swing2LimitDegrees=15, twistLimitDegrees=5}
 {bone1Name=dyn_cable_2, bone2Name=dyn_cable_3, swing1LimitDegrees=30, swing2LimitDegrees=30, twistLimitDegrees=10}
]
