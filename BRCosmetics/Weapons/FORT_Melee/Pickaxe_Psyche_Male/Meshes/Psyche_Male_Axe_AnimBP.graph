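# Post-processes the Psyche Male pickaxe skeleton, adding physics motion and bone offsets to its cable and glob bones.
# Each update turns bIsInFrontEnd into two opposite alphas that select between front-end and in-game RigidBody nodes, each with its own physics asset and motion scaling.
# The same alphas select a dyn_redCable_1 offset and scale, then fixed additive translations and scales adjust dyn_globA_2 through dyn_globM_2.
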AnimBlueprint /BRCosmetics/Weapons/FORT_Melee/Pickaxe_Psyche_Male/Meshes/Psyche_Male_Axe_AnimBP
target_skeleton=&/Psyche_Male_Axe_Skeleton
graphs [
 AnimGraph {nodes [
   n1 = Get isFrontEnd() -> (isFrontEnd:double)
   n2 = RigidBody(ComponentPose=Local To Component(LocalPose=Input Pose()), Alpha=@n1) {
    bClampLinearTranslationLimitToRefPose=true
    bOverrideWorldGravity=true
    overridePhysicsAsset=&/Psyche_Male_Axe_Physics_FrontEnd
    alphaBoolBlend {blendOption=Linear}
    baseBoneRef {boneName=Melee_TwoHanded_Handle}
    componentAppliedLinearAccClamp=<6000 6000 6000>
    componentLinearAccScale=<0.1 0.1 0.1>
    componentLinearVelScale=<0.1 0.1 0.1>
   }
   n3 = Get isNotFrontEnd() -> (isNotFrontEnd:double)
   n4 = RigidBody(ComponentPose=@n2, Alpha=@n3) {
    bClampLinearTranslationLimitToRefPose=true
    bOverrideWorldGravity=true
    overridePhysicsAsset=&/Psyche_Male_Axe_Physics
    alphaBoolBlend {blendOption=Linear}
    baseBoneRef {boneName=Melee_TwoHanded_Handle}
    componentAppliedLinearAccClamp=<3000 3000 3000>
    componentLinearAccScale=<0.03 0.03 0.03>
    componentLinearVelScale=<0.04 0.04 0.04>
   }
   n5 = "Transform (Modify) Bone - Bone: dyn_redCable_1"(ComponentPose=@n4, Alpha=@n1, Translation="0.5,0,0", Rotation="0,0,0", Scale="2,1,1") {
    scaleMode=BMM_Additive
    scaleSpace=BCS_BoneSpace
    translationMode=BMM_Additive
    translationSpace=BCS_BoneSpace
    alphaBoolBlend {blendOption=Linear}
    boneToModify {boneName=dyn_redCable_1}
    scale {x=2}
    translation {x=0.5}
   }
   n6 = "Transform (Modify) Bone - Bone: dyn_redCable_1"(ComponentPose=@n5, Alpha=@n3, Translation="-0.5,0,0", Rotation="0,0,0", Scale="1.3,1,1") {
    scaleMode=BMM_Additive
    scaleSpace=BCS_BoneSpace
    translationMode=BMM_Additive
    translationSpace=BCS_BoneSpace
    alphaBoolBlend {blendOption=Linear}
    boneToModify {boneName=dyn_redCable_1}
    scale {x=1.3}
    translation {x=-0.5}
   }
   n7 = "Transform (Modify) Bone - Bone: dyn_globA_2"(ComponentPose=@n6, Alpha=1, Translation="6.5,0,0", Rotation="0,0,0", Scale="2,0.5,0.5") {
    scaleMode=BMM_Additive
    scaleSpace=BCS_BoneSpace
    translationMode=BMM_Additive
    translationSpace=BCS_BoneSpace
    alphaBoolBlend {blendOption=Linear}
    boneToModify {boneName=dyn_globA_2}
    scale=<2 0.5 0.5>
    translation {x=6.5}
   }
   n8 = "Transform (Modify) Bone - Bone: dyn_globB_2"(ComponentPose=@n7, Alpha=1, Translation="5,0,0", Rotation="0,0,0", Scale="1,0.8,0.8") {
    scaleMode=BMM_Additive
    scaleSpace=BCS_BoneSpace
    translationMode=BMM_Additive
    translationSpace=BCS_BoneSpace
    alphaBoolBlend {blendOption=Linear}
    boneToModify {boneName=dyn_globB_2}
    scale {y=0.8, z=0.8}
    translation {x=5}
   }
   n9 = "Transform (Modify) Bone - Bone: dyn_globC_2"(ComponentPose=@n8, Alpha=1, Translation="6.5,0,0", Rotation="0,0,0", Scale="1,0.5,0.5") {
    scaleMode=BMM_Additive
    scaleSpace=BCS_BoneSpace
    translationMode=BMM_Additive
    translationSpace=BCS_BoneSpace
    alphaBoolBlend {blendOption=Linear}
    boneToModify {boneName=dyn_globC_2}
    scale {y=0.5, z=0.5}
    translation {x=6.5}
   }
   n10 = "Transform (Modify) Bone - Bone: dyn_globD_2"(ComponentPose=@n9, Alpha=1, Translation="5,0,0", Rotation="0,0,0", Scale="1,0.5,0.5") {
    scaleMode=BMM_Additive
    scaleSpace=BCS_BoneSpace
    translationMode=BMM_Additive
    translationSpace=BCS_BoneSpace
    alphaBoolBlend {blendOption=Linear}
    boneToModify {boneName=dyn_globD_2}
    scale {y=0.5, z=0.5}
    translation {x=5}
   }
   n11 = "Transform (Modify) Bone - Bone: dyn_globE_2"(ComponentPose=@n10, Alpha=1, Translation="7,0,0", Rotation="0,0,0", Scale="1,0.5,0.5") {
    scaleMode=BMM_Additive
    scaleSpace=BCS_BoneSpace
    translationMode=BMM_Additive
    translationSpace=BCS_BoneSpace
    alphaBoolBlend {blendOption=Linear}
    boneToModify {boneName=dyn_globE_2}
    scale {y=0.5, z=0.5}
    translation {x=7}
   }
   n12 = "Transform (Modify) Bone - Bone: dyn_globF_2"(ComponentPose=@n11, Alpha=1, Translation="5,0,0", Rotation="0,0,0", Scale="1,0.5,0.5") {
    scaleMode=BMM_Additive
    scaleSpace=BCS_BoneSpace
    translationMode=BMM_Additive
    translationSpace=BCS_BoneSpace
    alphaBoolBlend {blendOption=Linear}
    boneToModify {boneName=dyn_globF_2}
    scale {y=0.5, z=0.5}
    translation {x=5}
   }
   n13 = "Transform (Modify) Bone - Bone: dyn_globG_2"(ComponentPose=@n12, Alpha=1, Translation="8,0,0", Rotation="0,0,0", Scale="1,0.5,0.5") {
    scaleMode=BMM_Additive
    scaleSpace=BCS_BoneSpace
    translationMode=BMM_Additive
    translationSpace=BCS_BoneSpace
    alphaBoolBlend {blendOption=Linear}
    boneToModify {boneName=dyn_globG_2}
    scale {y=0.5, z=0.5}
    translation {x=8}
   }
   n14 = "Transform (Modify) Bone - Bone: dyn_globH_2"(ComponentPose=@n13, Alpha=1, Translation="7,0,0", Rotation="0,0,0", Scale="1,0.5,0.5") {
    scaleMode=BMM_Additive
    scaleSpace=BCS_BoneSpace
    translationMode=BMM_Additive
    translationSpace=BCS_BoneSpace
    alphaBoolBlend {blendOption=Linear}
    boneToModify {boneName=dyn_globH_2}
    scale {y=0.5, z=0.5}
    translation {x=7}
   }
   n15 = "Transform (Modify) Bone - Bone: dyn_globI_2"(ComponentPose=@n14, Alpha=1, Translation="6,0,0", Rotation="0,0,0", Scale="1,0.5,0.5") {
    scaleMode=BMM_Additive
    scaleSpace=BCS_BoneSpace
    translationMode=BMM_Additive
    translationSpace=BCS_BoneSpace
    alphaBoolBlend {blendOption=Linear}
    boneToModify {boneName=dyn_globI_2}
    scale {y=0.5, z=0.5}
    translation {x=6}
   }
   n16 = "Transform (Modify) Bone - Bone: dyn_globJ_2"(ComponentPose=@n15, Alpha=1, Translation="5,0,0", Rotation="0,0,0", Scale="1,0.5,0.5") {
    scaleMode=BMM_Additive
    scaleSpace=BCS_BoneSpace
    translationMode=BMM_Additive
    translationSpace=BCS_BoneSpace
    alphaBoolBlend {blendOption=Linear}
    boneToModify {boneName=dyn_globJ_2}
    scale {y=0.5, z=0.5}
    translation {x=5}
   }
   n17 = "Transform (Modify) Bone - Bone: dyn_globK_2"(ComponentPose=@n16, Alpha=1, Translation="8,0,0", Rotation="0,0,0", Scale="1,0.5,0.5") {
    scaleMode=BMM_Additive
    scaleSpace=BCS_BoneSpace
    translationMode=BMM_Additive
    translationSpace=BCS_BoneSpace
    alphaBoolBlend {blendOption=Linear}
    boneToModify {boneName=dyn_globK_2}
    scale {y=0.5, z=0.5}
    translation {x=8}
   }
   n18 = "Transform (Modify) Bone - Bone: dyn_globL_2"(ComponentPose=@n17, Alpha=1, Translation="8,0,0", Rotation="0,0,0", Scale="1,0.5,0.5") {
    scaleMode=BMM_Additive
    scaleSpace=BCS_BoneSpace
    translationMode=BMM_Additive
    translationSpace=BCS_BoneSpace
    alphaBoolBlend {blendOption=Linear}
    boneToModify {boneName=dyn_globL_2}
    scale {y=0.5, z=0.5}
    translation {x=8}
   }
   n19 = "Transform (Modify) Bone - Bone: dyn_globM_2"(ComponentPose=@n18, Alpha=1, Translation="5,0,0", Rotation="0,0,0", Scale="1,0.5,0.5") {
    scaleMode=BMM_Additive
    scaleSpace=BCS_BoneSpace
    translationMode=BMM_Additive
    translationSpace=BCS_BoneSpace
    alphaBoolBlend {blendOption=Linear}
    boneToModify {boneName=dyn_globM_2}
    scale {y=0.5, z=0.5}
    translation {x=5}
   }
   Output Pose(Result=Component To Local(ComponentPose=@n19)) {layerGroup=None}
  ]}
 EventGraph {nodes [
   Event BlueprintUpdateAnimation()
   > n1 = Branch(Condition=Get bIsInFrontEnd())
   > Set isFrontEnd(isFrontEnd:double=1) -> (Output_Get:double)
   > Set isNotFrontEnd(isNotFrontEnd:double=0) -> (Output_Get:double)
   Set isFrontEnd(execute=@n1.else, isFrontEnd=0)
   > Set isNotFrontEnd(isNotFrontEnd=1)
   TryGetPawnOwner()
   TryGetPawnOwner()
  ]}
]
variables [
 isFrontEnd {type=number}
 isJumping {type=number}
 isNotFrontEnd {type=number}
]
